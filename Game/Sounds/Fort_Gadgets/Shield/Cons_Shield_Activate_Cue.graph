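# Sound cue for activating the shield consumable gadget, built by reusing the Electric Trap placement sound.
# A quality level node sends its first two levels to the standard wave at double pitch and the last two to an "_opt" wave at a smaller pitch raise.
# It uses super-close player attenuation and an important in-game sound class, with a concurrency override that prevents new instances once the limit is reached.
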SoundCue /Game/Sounds/Fort_Gadgets/Shield/Cons_Shield_Activate_Cue
root=QualityLevel_0
nodes [
 {id=Modulator_0, children [WavePlayer_12], properties {pitchMax=2, pitchMin=2}}
 {id=Modulator_2, children [WavePlayer_13], properties {pitchMax=1.5, pitchMin=1.5}}
 {id=QualityLevel_0, children [Modulator_0,Modulator_0,Modulator_2,Modulator_2]}
 {id=WavePlayer_12, properties {soundWaveAssetPtr=/Game/Sounds/Fort_Traps/Electric_Trap/Electric_Trap_Place01}}
 {id=WavePlayer_13, properties {soundWaveAssetPtr=/Game/Sounds/Fort_Traps/Electric_Trap/Electric_Trap_Place01_opt}}
]
properties {
 attenuationSettings=/Game/Sounds/Attenuation/Fort_Player_SuperClose_Att
 bOverrideConcurrency=true
 maxDistance=800
 soundClassObject=&Fort_Mission_Ingame_Important_PSM
 volumeMultiplier=0.7
 concurrencyOverrides {resolutionRule=PreventNew}
}
