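# First-person sound cue for shield deactivation in Athena, playing the shield-break-received hit marker sound.
# A single wave player feeds the output directly at reduced volume, with no randomization or modulation.
# It routes through the 2D received shield-break hit notification sound class.
# Its concurrency set is the shared critical hit-notification stop-oldest rule.
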SoundCue /Game/Sounds/Fort_Gadgets/Shield/Cons_Shield_Deactivate_Athena_1P_Cue
root=WavePlayer_0
nodes [{id=WavePlayer_0, properties {soundWaveAssetPtr=/Game/Athena/Sounds/HitNotifications/HitMarker_ShieldBreak_Receive_01}}]
properties {
 maxDistance=2097152
 soundClassObject=/Game/Sounds/FortSoundClassesAndModes/Fort_2DHitNotifications_Receive_Shield_Break_PSM
 volumeMultiplier=0.6
 concurrencyOverrides {maxCount=1, resolutionRule=StopLowestPriority}
 concurrencySet [/Game/Sounds/Concurrency_Rules/HitNotifications_Critical_StopOldest]
}
