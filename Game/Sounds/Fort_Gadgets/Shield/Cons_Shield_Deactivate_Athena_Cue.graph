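# Sound cue for a shield deactivation in Athena, playing a 3D shield-break hit marker sound.
# A single wave player feeds a Modulator that randomizes pitch while holding volume constant.
# It routes to the impacts sound class with shield-break 3D attenuation.
# A concurrency override caps it at two instances, stopping the lowest priority, and takes the place of the referenced hit-notification concurrency set.
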SoundCue /Game/Sounds/Fort_Gadgets/Shield/Cons_Shield_Deactivate_Athena_Cue
root=Modulator_1
nodes [
 {id=Modulator_1, children [WavePlayer_0], properties {pitchMax=1.1, volumeMax=1, volumeMin=1}}
 {id=WavePlayer_0, properties {soundWaveAssetPtr=/Game/Athena/Sounds/HitNotifications/HitMarker_ShieldBreak_3D_01}}
]
properties {
 attenuationSettings=/Game/Sounds/Attenuation/Fort_Shield_Break_3D_Att
 bOverrideConcurrency=true
 maxDistance=8500
 soundClassObject=&Fort_Impacts
 volumeMultiplier=0.9375
 concurrencyOverrides {maxCount=2, resolutionRule=StopLowestPriority}
 concurrencySet [&HitNotifications_TwoMaximum_StopOldest_VolumeScale]
}
