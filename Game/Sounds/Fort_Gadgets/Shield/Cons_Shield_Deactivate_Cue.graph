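# Sound cue for shield gadget deactivation, classed as a health-related player sound.
# A Mixer layers a shield power-down wave, through a Modulator with pitch pinned at normal, over a borrowed electric ceiling-trap firing sound.
# An Enveloper fades the electric layer out over its first second.
# It plays at half volume with super-close player attenuation, and its concurrency override prevents new instances once the limit is reached.
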SoundCue /Game/Sounds/Fort_Gadgets/Shield/Cons_Shield_Deactivate_Cue
root=Mixer_1
nodes [
 {
  id=Enveloper_0
  children [WavePlayer_27]
  properties {pitchCurve {editorCurveData {keys [{time=0}]}}, volumeCurve {editorCurveData {keys [<at 0 1>,<at 1 0>]}}}
 }
 {id=Mixer_1, children [Modulator_0,Enveloper_0], properties {inputVolume [1,1]}}
 {id=Modulator_0, children [WavePlayer_22], properties {pitchMax=1, pitchMin=1}}
 {id=WavePlayer_22, properties {soundWaveAssetPtr=&/cons_shield_powerdown}}
 {id=WavePlayer_27, properties {soundWaveAssetPtr=/Game/Sounds/Fort_Traps/Electric_Trap/CeilingTrap_Electric_Fire_002}}
]
properties {
 attenuationSettings=/Game/Sounds/Attenuation/Fort_Player_SuperClose_Att
 bOverrideConcurrency=true
 maxDistance=800
 soundClassObject=/Game/Sounds/FortSoundClassesAndModes/Fort_03_Player_Health_Related
 volumeMultiplier=0.5
 concurrencyOverrides {resolutionRule=PreventNew}
}
