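# Sound cue for the full shield consumable gadget, played quietly with super-close player attenuation.
# A quality-level root sends its first two slots to the standard wave and its last two to the "opt" variant.
# Each wave runs through its own modulator with a fixed raised pitch, so there is no random variation.
# An overridden concurrency rule prevents new instances once the limit is reached.
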SoundCue /Game/Sounds/Fort_Gadgets/Shield/Cons_Shield_Full_Cue
root=QualityLevel_0
nodes [
 {id=Modulator_0, children [WavePlayer_1], properties {pitchMax=1.25, pitchMin=1.25}}
 {id=Modulator_1, children [WavePlayer_2], properties {pitchMax=1.25, pitchMin=1.25}}
 {id=QualityLevel_0, children [Modulator_0,Modulator_0,Modulator_1,Modulator_1]}
 {id=WavePlayer_1, properties {soundWaveAssetPtr=&/Cons_Shield_Full_03}}
 {id=WavePlayer_2, properties {soundWaveAssetPtr=&/Cons_Shield_Full_03_opt}}
]
properties {
 attenuationSettings=/Game/Sounds/Attenuation/Fort_Player_SuperClose_Att
 bOverrideConcurrency=true
 maxDistance=800
 soundClassObject=&Fort_Mission_Ingame_Important_PSM
 volumeMultiplier=0.25
 concurrencyOverrides {resolutionRule=PreventNew}
}
