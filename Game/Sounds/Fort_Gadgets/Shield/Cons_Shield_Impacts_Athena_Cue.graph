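# Sound cue for shield impacts in Athena, played as a random one-shot from a set of shield impact waves.
# A Modulator feeds a quality-level switch: the first level picks evenly from ten waves, the second from three, and the other two slots are empty.
# It uses bullet-impact attenuation, the impacts sound class, and a concurrency rule named for one sound per owner, stopping the oldest.
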SoundCue /Game/Sounds/Fort_Gadgets/Shield/Cons_Shield_Impacts_Athena_Cue
root=Modulator_0
nodes [
 {id=Modulator_0, children [QualityLevel_0]}
 {id=QualityLevel_0, children [Random_0,Random_3,null,null]}
 {
  id=Random_0
  children [WavePlayer_0,WavePlayer_1,WavePlayer_2,WavePlayer_3,WavePlayer_4,WavePlayer_5,WavePlayer_6,WavePlayer_7,WavePlayer_8,WavePlayer_9]
  properties {weights [1,1,1,1,1,1,1,1,1,1]}
 }
 {id=Random_3, children [WavePlayer_22, WavePlayer_25, WavePlayer_27], properties {weights [1,1,1]}}
 {id=WavePlayer_0, properties {soundWaveAssetPtr=&/shield_impact_03}}
 {id=WavePlayer_1, properties {soundWaveAssetPtr=&/shield_impact_07}}
 {id=WavePlayer_2, properties {soundWaveAssetPtr=&/shield_impact_10}}
 {id=WavePlayer_22, properties {soundWaveAssetPtr=&shield_impact_03}}
 {id=WavePlayer_25, properties {soundWaveAssetPtr=&/shield_impact_05}}
 {id=WavePlayer_27, properties {soundWaveAssetPtr=&/shield_impact_04}}
 {id=WavePlayer_3, properties {soundWaveAssetPtr=&/shield_impact_01}}
 {id=WavePlayer_4, properties {soundWaveAssetPtr=&/shield_impact_02}}
 {id=WavePlayer_5, properties {soundWaveAssetPtr=&shield_impact_04}}
 {id=WavePlayer_6, properties {soundWaveAssetPtr=&shield_impact_05}}
 {id=WavePlayer_7, properties {soundWaveAssetPtr=&/shield_impact_06}}
 {id=WavePlayer_8, properties {soundWaveAssetPtr=&/shield_impact_08}}
 {id=WavePlayer_9, properties {soundWaveAssetPtr=&/shield_impact_09}}
]
properties {
 attenuationSettings=/Game/Sounds/Attenuation/Fort_Bullet_Impact_Att
 maxDistance=5300
 soundClassObject=&Fort_Impacts
 concurrencyOverrides {resolutionRule=PreventNew}
 concurrencySet [/Game/Sounds/Concurrency_Rules/OneMaximum_Owner_StopOldest]
}
