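# Sound cue for shield impact sounds in the gadgets audio set, played on the Fort_Impacts sound class with super-close player attenuation.
# A quality-level root selects an equal-weight random pick from ten waves, one from a seven-wave subset, or one of two modulated single-wave branches using impact 10.
# This scales variation down across quality levels, and a concurrency set limits playback to one per owner, stopping the oldest.
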SoundCue /Game/Sounds/Fort_Gadgets/Shield/Cons_Shield_Impacts_Cue
root=QualityLevel_0
nodes [
 {id=Modulator_0, children [WavePlayer_0]}
 {id=Modulator_1, children [WavePlayer_1]}
 {id=QualityLevel_0, children [Random_1,Random_2,Modulator_0,Modulator_1]}
 {
  id=Random_1
  children [
   WavePlayer_10
   WavePlayer_11
   WavePlayer_12
   WavePlayer_13
   WavePlayer_14
   WavePlayer_15
   WavePlayer_16
   WavePlayer_17
   WavePlayer_18
   WavePlayer_19
  ]
  properties {weights [1,1,1,1,1,1,1,1,1,1]}
 }
 {
  id=Random_2
  children [WavePlayer_20, WavePlayer_21, WavePlayer_23, WavePlayer_24, WavePlayer_26, WavePlayer_28, WavePlayer_29]
  properties {weights [1,1,1,1,1,1,1]}
 }
 {id=WavePlayer_0, properties {soundWaveAssetPtr=&/cons_shield_impact_10}}
 {id=WavePlayer_1, properties {soundWaveAssetPtr=&cons_shield_impact_10}}
 {id=WavePlayer_10, properties {soundWaveAssetPtr=&cons_shield_impact_10}}
 {id=WavePlayer_11, properties {soundWaveAssetPtr=&/cons_shield_impact_01}}
 {id=WavePlayer_12, properties {soundWaveAssetPtr=&/cons_shield_impact_02}}
 {id=WavePlayer_13, properties {soundWaveAssetPtr=&/cons_shield_impact_03}}
 {id=WavePlayer_14, properties {soundWaveAssetPtr=&/cons_shield_impact_04}}
 {id=WavePlayer_15, properties {soundWaveAssetPtr=&/cons_shield_impact_05}}
 {id=WavePlayer_16, properties {soundWaveAssetPtr=&/cons_shield_impact_06}}
 {id=WavePlayer_17, properties {soundWaveAssetPtr=&/cons_shield_impact_07}}
 {id=WavePlayer_18, properties {soundWaveAssetPtr=&/cons_shield_impact_08}}
 {id=WavePlayer_19, properties {soundWaveAssetPtr=&/cons_shield_impact_09}}
 {id=WavePlayer_20, properties {soundWaveAssetPtr=&cons_shield_impact_10}}
 {id=WavePlayer_21, properties {soundWaveAssetPtr=&cons_shield_impact_01}}
 {id=WavePlayer_23, properties {soundWaveAssetPtr=&cons_shield_impact_03}}
 {id=WavePlayer_24, properties {soundWaveAssetPtr=&cons_shield_impact_04}}
 {id=WavePlayer_26, properties {soundWaveAssetPtr=&cons_shield_impact_06}}
 {id=WavePlayer_28, properties {soundWaveAssetPtr=&cons_shield_impact_08}}
 {id=WavePlayer_29, properties {soundWaveAssetPtr=&cons_shield_impact_09}}
]
properties {
 attenuationSettings=/Game/Sounds/Attenuation/Fort_Player_SuperClose_Att
 maxDistance=800
 soundClassObject=&Fort_Impacts
 concurrencyOverrides {resolutionRule=PreventNew}
 concurrencySet [/Game/Sounds/Concurrency_Rules/OneMaximum_Owner_StopOldest]
}
