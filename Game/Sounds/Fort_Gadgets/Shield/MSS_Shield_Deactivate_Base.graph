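# Stereo one-shot MetaSound base for shield deactivation, defaulting to a received shield-break hit-marker wave.
# Trigger compares check time since last play against a threshold to choose normal or rapid-retrigger playback.
# The retrigger path routes a gain adjustment and start-time offset into a deluxe wave player, plus a fade-in multiplied onto both channels.
# The player also randomizes gain and pitch, while concurrency uses a stop-oldest hit-notification rule.
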MetaSoundSource /Game/Sounds/Fort_Gadgets/Shield/MSS_Shield_Deactivate_Base
inputs [
 Retrigger.BehaviorThresholdTime {type=Float, access=Constructor}
 Retrigger.FadeInTime {type=Float, access=Constructor}
 "Retrigger.GainAdjustment (Db)" {type=Float, access=Constructor}
 Retrigger.TimeOffset {type=Time, access=Constructor}
 "Source.Gain (Db)" {type=Float}
 "Source.Pitch (Semitone)" {type=Float}
 Source.RandomGain.Max {type=Float, access=Constructor}
 Source.RandomGain.Min {type=Float, access=Constructor}
 Source.RandomPitch.Max {type=Float, access=Constructor}
 Source.RandomPitch.Min {type=Float, access=Constructor}
 Source.Waves {type=WaveAsset:Array, access=Constructor}
 Time.ElapsedTime {type=Float}
 Time.PlayTime {type=Float}
 Time.TimeSinceLastPlayed {type=Float}
 UE.Source.OnPlay {type=Trigger}
]
interfaces [Fort.Time:1.0, UE.OutputFormat.Stereo:1.0, UE.Source.OneShot:1.0, UE.Source:1.1]
outputs [
 "UE.OutputFormat.Stereo.Audio:0" {type=Audio}
 "UE.OutputFormat.Stereo.Audio:1" {type=Audio}
 UE.Source.OneShot.OnFinished {type=Trigger}
]
pages [Default {
  input_defaults {
   Retrigger.BehaviorThresholdTime=0.75
   Retrigger.FadeInTime=0.1
   "Retrigger.GainAdjustment (Db)"=-1
   Retrigger.TimeOffset=0.1
   "Source.Gain (Db)"=0
   "Source.Pitch (Semitone)"=0
   Source.RandomGain.Max=0
   Source.RandomGain.Min=0
   Source.RandomPitch.Max=0
   Source.RandomPitch.Min=0
   Time.ElapsedTime=0
   Time.PlayTime=0
   Time.TimeSinceLastPlayed=0
   UE.Source.OnPlay=null
   Source.Waves [/Game/Athena/Sounds/HitNotifications/HitMarker_ShieldBreak_Receive_01]
  }
  nodes [
   InitVariable.WaveAsset:Array()
   n1 = TriggerCompare.Clamp:Float(A=^Time.TimeSinceLastPlayed, Compare=^UE.Source.OnPlay)
   n2 = TriggerCompare.Clamp:Float(B=^Retrigger.BehaviorThresholdTime, A=^Time.TimeSinceLastPlayed, Compare=@n1.FALSE) {also {defaults {Type=3}}}
   n3 = VariableAccessor.Trigger(Variable=VariableMutator.Trigger(Variable=InitVariable.Trigger().Variable, Value=@n2.FALSE).Variable)
   TriggerAnyTriggerAny2_0 = "TriggerAny.Trigger Any (2)"(In 0=@n1.TRUE, In 1=@n2.TRUE)
   n4 = VariableAccessor.Trigger(Variable=VariableMutator.Trigger(Variable=InitVariable.Trigger().Variable, Value=@TriggerAnyTriggerAny2_0.out).Variable)
   TriggerRouteTriggerRouteFloat2_0 = "TriggerRoute.Trigger Route (Float, 2):Float"(Value 1=^"Retrigger.GainAdjustment (Db)", Set 1=@n3.Value, Set 0=@n4.Value)
   n5 = VariableAccessor.Trigger(Variable=@n4.Variable)
   n6 = VariableAccessor.Trigger(Variable=@n3.Variable)
   TriggerRouteTriggerRouteTime2_0 = "TriggerRoute.Trigger Route (Time, 2):Time"(
    Value 1=^Retrigger.TimeOffset
    Set 0=VariableAccessor.Trigger(Variable=@n5.Variable).Value
    Set 1=VariableAccessor.Trigger(Variable=@n6.Variable).Value
   )
   TriggerAccumulatorTriggerAccumulate2_0 = "TriggerAccumulator.Trigger Accumulate (2)"(In 0=@TriggerRouteTriggerRouteFloat2_0.On Set, In 1=@TriggerRouteTriggerRouteTime2_0.On Set)
   MSP_Audio_WavePlayer_Deluxe_Stereo_0 = &MSP_Audio_WavePlayer_Deluxe_Stereo(
    "Pitch Shift Global (Semitones)"=^"Source.Pitch (Semitone)"
    Max Random Decibels=^Source.RandomGain.Max
    Min Random Decibels=^Source.RandomGain.Min
    "Random Pitch Max (Semitones)"=^Source.RandomPitch.Max
    "Random Pitch Min (Semitones)"=^Source.RandomPitch.Min
    Wave Asset=^Source.Waves
    WavePlayer Gain dB=UE.Reroute(Value=UE.Add:Float(PrimaryOperand=^"Source.Gain (Db)", AdditionalOperands=@TriggerRouteTriggerRouteFloat2_0.Value).out).Value
    Start Time Global Offset=UE.Reroute(Value=UE.Reroute(Value=@TriggerRouteTriggerRouteTime2_0.Value).Value).Value
    Play Trigger=@TriggerAccumulatorTriggerAccumulate2_0.out
   )
   TriggerRouteTriggerRouteFloat2_1 = "TriggerRoute.Trigger Route (Float, 2):Float"(Value 1=^Retrigger.FadeInTime, Set 0=@n5.Value, Set 1=@n6.Value) {also {defaults {Value 0=0}}}
   n7 = &MSP_Fade(In=@TriggerRouteTriggerRouteFloat2_1.On Set, Fade Time=@TriggerRouteTriggerRouteFloat2_1.Value)
   Multiply_0 = UE.Multiply:Audio by Float(PrimaryOperand=@MSP_Audio_WavePlayer_Deluxe_Stereo_0.Out L, AdditionalOperands=@n7.Value Out)
   Multiply_1 = UE.Multiply:Audio by Float(PrimaryOperand=@MSP_Audio_WavePlayer_Deluxe_Stereo_0.Out R, AdditionalOperands=@n7.Value Out)
  ]
  links [
   {dst="@UE.Source.OneShot.OnFinished", src=MSP_Audio_WavePlayer_Deluxe_Stereo_0.On Finished}
   {dst="@UE.OutputFormat.Stereo.Audio:0", src=Multiply_0.out}
   {dst="@UE.OutputFormat.Stereo.Audio:1", src=Multiply_1.out}
  ]
 }]
properties {
 assetImportData=&self:AssetImportData
 numChannels=2
 outputFormat=Stereo
 sampleRate=48000
 soundClassObject=/Game/Sounds/FortSoundClassesAndModes/Fort_2DHitNotifications_Receive_Shield_Break_PSM
 concurrencySet [/Game/Sounds/Concurrency_Rules/HitNotifications_Critical_StopOldest]
}
