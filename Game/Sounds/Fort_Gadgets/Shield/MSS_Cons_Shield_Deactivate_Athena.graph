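# MetaSound Source preset of MSS_Shield_Deactivate_Base that plays the Athena shield deactivation one-shot, using a 3D shield-break hit marker wave.
# The parent supplies the graph; this preset feeds it the wave, gain, pitch and retrigger settings (time threshold, fade-in, offset, gain trim), alongside Fort.Time timing inputs.
# Stereo output uses observed hit-marker attenuation and the impacts sound class, with a concurrency override stopping lowest-priority voices.
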MetaSoundSource /Game/Sounds/Fort_Gadgets/Shield/MSS_Cons_Shield_Deactivate_Athena
preset_parent=&/MSS_Shield_Deactivate_Base
inputs [
 Retrigger.BehaviorThresholdTime {type=Float, access=Constructor}
 Retrigger.FadeInTime {type=Float, access=Constructor}
 "Retrigger.GainAdjustment (Db)" {type=Float, access=Constructor}
 Retrigger.TimeOffset {type=Time, access=Constructor}
 "Source.Gain (Db)" {type=Float}
 "Source.Pitch (Semitone)" {type=Float}
 Source.RandomGain.Max {type=Float, access=Constructor}
 Source.RandomGain.Min {type=Float, access=Constructor}
 Source.RandomPitch.Max {type=Float, access=Constructor}
 Source.RandomPitch.Min {type=Float, access=Constructor}
 Source.Waves {type=WaveAsset:Array, access=Constructor}
 Time.ElapsedTime {type=Float}
 Time.PlayTime {type=Float}
 Time.TimeSinceLastPlayed {type=Float}
 UE.Source.OnPlay {type=Trigger}
]
interfaces [Fort.Time:1.0, UE.OutputFormat.Stereo:1.0, UE.Source.OneShot:1.0, UE.Source:1.1]
outputs [
 "UE.OutputFormat.Stereo.Audio:0" {type=Audio}
 "UE.OutputFormat.Stereo.Audio:1" {type=Audio}
 UE.Source.OneShot.OnFinished {type=Trigger}
]
pages [Default {input_defaults {
   Retrigger.BehaviorThresholdTime=0.75
   Retrigger.FadeInTime=0.1
   "Retrigger.GainAdjustment (Db)"=-1
   Retrigger.TimeOffset=0.15
   "Source.Gain (Db)"=3
   "Source.Pitch (Semitone)"=0
   Source.RandomGain.Max=0
   Source.RandomGain.Min=0
   Source.RandomPitch.Max=1
   Source.RandomPitch.Min=1
   Time.ElapsedTime=0
   Time.PlayTime=0
   Time.TimeSinceLastPlayed=0
   UE.Source.OnPlay=null
   Source.Waves [/Game/Athena/Sounds/HitNotifications/HitMarker_ShieldBreak_3D_01]
  }}]
properties {
 assetImportData=&self:AssetImportData
 attenuationSettings=/Game/Athena/Sounds/HitNotifications/Fort_HitMarker_Observed_Att
 bOverrideConcurrency=true
 maxDistance=5500
 numChannels=2
 outputFormat=Stereo
 sampleRate=48000
 soundClassObject=&Fort_Impacts
 volume=0.75
 concurrencyOverrides {maxCount=2, resolutionRule=StopLowestPriority}
 concurrencySet [/Game/Sounds/Concurrency_Rules/HitNotifications_Critical_StopOldest]
}
